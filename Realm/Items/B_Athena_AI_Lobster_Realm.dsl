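; Realm AI lightsaber weapon actor for Vader that extends a parent saber, with blade material tweaks and a Throw/Attack HUD prompt toggle.
; On BeginPlay, a valid base Mantis component is replaced with a Realm AI version to override animations.
; The server then hides the blade when the instigator starts skytubing.
; Clients add a loose gameplay tag to hide the belt hilt, removing it on EndPlay.
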(event EventBeginPlay
  (bind _self self)
  (bind _returnvalue (Actor|GetComponentbyClass _self "/Galileo/Mantis/MantisWeaponComp_GalileoLobster.MantisWeaponComp_GalileoLobster_C"))
  ;[1 1 1] Remove the previous Mantis Component and add the new one so we can override the animations that is used by Mantis.
  (Utilities|IsValid _returnvalue
    (:Is%20Valid
      (Components|DestroyComponent _returnvalue)
      (Game|AddComponentbyClass _self "/Realm/Mantis/MantisWeaponComp_GalileoLobsterRealmAI.MantisWeaponComp_GalileoLobsterRealmAI_C")
      (|CallParentFunction)
      ;[1 1 1] Add a Tag on the Client to Hide the Hilt on the Belt
      (Networking|SwitchHasAuthority
        (:Authority
          ;[1 1 1] Fix for Vader's Lightsaber remaining in Skytubes
          (Ability|Tasks|WaitGameplayTagAddToActor (Game|GetInstigator) "(TagName=\"Gameplay.Status.Skytubing\")"
            (:Added
              (Variables|Default|SetBladeVisibility false))))
        (:Remote
          (Ability|GameplayEffect|AddLooseGameplayTags (Game|GetInstigator) (Variables|Default|GetLobsterRealmWeaponTag)))))))

(event EventEndPlay
  (params (EndPlayReason :type EEndPlayReason%20Enum :default "Destroyed"))
  ;[1 1 1] On the Client, this will end when Vader holsters the weapon, so re-show the Hilt on the belt
  (Networking|SwitchHasAuthority
    (:Remote
      (Ability|GameplayEffect|RemoveLooseGameplayTags (Game|GetInstigator) (Variables|Default|GetLobsterRealmWeaponTag)))))

(event EventSetBladeVisibility
  (params (Visible :type Boolean))
  (|CallParentFunction Visible))

(event Custom|ToggleSaberThrowPrompt
  (params (Visible :type Boolean))
  (WeaponHUDData|SetHudKeyActionDescription 0 (select Visible "NSLOCTEXT(\"[A935DAEB490105970B2B73A67B41C217]\", \"10C60EF845236FE4297B688B582B2758\", \"Throw\")" "NSLOCTEXT(\"[A935DAEB490105970B2B73A67B41C217]\", \"00C7B8F7411632AB2CC31C9F36BF25A9\", \"Attack\")")))

(fn OnRep_ToggleBeltHiltByVisibility
  (bind _returnvalue (Game|GetInstigator))
  (bind _lobsterrealmweapontag (Variables|Default|GetLobsterRealmWeaponTag))
  (Networking|SwitchHasAuthority
    (:Remote
      (if true
        (Ability|GameplayEffect|AddLooseGameplayTags _returnvalue _lobsterrealmweapontag)
        (else
          (Ability|GameplayEffect|RemoveLooseGameplayTags _returnvalue _lobsterrealmweapontag))))))

(fn ConstructionScript
  (|CallParentFunction)
  (Rendering|Material|SetScalarParameterValueonMaterials (Variables|Default|GetSM_Galileo_Lobster_Blade_Masked(VisibleUnderwater)) "HueColorShift" (Variables|Default|GetBladeHueColorShift))
  (Rendering|Material|SetScalarParameterValueonMaterials (Variables|Default|GetSM_Galileo_Lobster_Blade) "Moblie_EmissiveIntensity" 4.0))
