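; Blueprint for the Realm variant of the Galileo Lobster item, which adjusts the materials on its blade meshes.
; The construction script runs the parent's version, then drives HueColorShift on the masked, underwater-visible blade mesh from the BladeHueColorShift variable.
; It also sets a fixed mobile emissive intensity on the main blade mesh, using the misspelled parameter name "Moblie_EmissiveIntensity".
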(fn ConstructionScript
  (|CallParentFunction)
  (Rendering|Material|SetScalarParameterValueonMaterials (Variables|Default|GetSM_Galileo_Lobster_Blade_Masked(VisibleUnderwater)) "HueColorShift" (Variables|Default|GetBladeHueColorShift))
  (Rendering|Material|SetScalarParameterValueonMaterials (Variables|Default|GetSM_Galileo_Lobster_Blade) "Moblie_EmissiveIntensity" 5e+01))
